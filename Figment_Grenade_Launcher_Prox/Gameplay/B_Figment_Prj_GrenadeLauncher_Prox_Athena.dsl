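; Proximity grenade launcher projectile that bounces and explodes near or on contact with enemy pawns or tagged actors.
; The server starts arm and max-lifetime timers; arming sets a replicated flag whose OnRep enables the overlap sphere.
; Bounces play surface-based effects and add gravity, detonating below a minimum speed; fuse expiry and water entry also detonate.
; The construction script reads sphere radius, gravity and FX scale from scalable floats, and DoOnce gates the explosion.
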(event Custom|Arm
  ;[1 1 1] Arm After Duration
  (Variables|Default|SetRepCollision true))

(event EventBeginPlay
  (if (Networking|IsServer)
    ;[1 1 1] Arm After Duration
    (Utilities|Time|SetTimerbyEvent (delegate Custom|Arm) (Variables|Default|GetArmTime))
    ;[1 1 1] Set Fuse
    (Utilities|Time|SetTimerbyFunctionName self "FuseTimerMax" (ScalableFloat|GetValueatLevel (Variables|Rows|GetRow_MaxLifetime)))))

(event EventOnStop
  (params (Hit :type Hit%20Result%20Structure%20%28by%20ref%29))
  (bind (_bblockinghit _binitialoverlap _time _distance _location _impactpoint _normal _impactnormal _physmat) (Collision|BreakHitResult Hit))
  ;[1 1 1] StopBehavior, get PhysMat
  (Variables|Default|SetStopLocation _impactpoint) ; Use Loc and Rot for single explosion
  (Variables|Default|SetStopRotX (Math|Rotator|MakeRotfromX _impactnormal))
  (Variables|Default|SetStopRotZ (Math|Rotator|MakeRotfromZ _impactnormal))
  (Utilities|IsValid _physmat
    (:Is%20Valid
      (Variables|Effects|SetStopSurfaceType (Variables|/Script/PhysicsCore.PhysicalMaterial|GetSurfaceType _physmat)))
    (:Is%20Not%20Valid
      _))
  (if (Utilities|Enum|Equal(Enum) (Variables|Effects|GetStopSurfaceType) "SurfaceType8")
    (Rendering|SetVisibility (Variables|Projectile_GrenadeLauncher|GetMesh) false false)
    (bind _returnvalue (Effects|SpawnEmitteratLocation (Variables|Effects|GetWaterLandDelayPS) (Variables|Default|GetStopLocation) (Variables|Default|GetStopRotX)))
    (Variables|Default|SetWaterLandDelayRipplePSC _returnvalue)))

(event Custom|FuseTimerMax
  ;[1 1 1] Bounce
  ;;[1 1 1] DelayExplosion
  (if (not (Delivery|IsBeingKilled))
    (Components|Activation|SetActive (Variables|Default|GetTrail) false false)
    (CallFunction|DoExplosionOnce)))

(event EventOnBounce
  (params (Hit :type Hit%20Result%20Structure%20%28by%20ref%29))
  (bind (_bblockinghit _binitialoverlap _time _distance _location _impactpoint _normal _impactnormal _physmat) (Collision|BreakHitResult Hit))
  (bind _projectilemovementcomponent (Variables|Movement|GetProjectileMovementComponent))
  ;[1 1 1] Bounce
  (|CallParentFunction Hit)
  (Variables|Default|SetOnBounceHitResult Hit)
  (Utilities|FlowControl|Sequence
    (:then_0
      (Variables|Default|SetStopLocation _impactpoint) ; Use Loc and Rot for single explosion
      (Variables|Default|SetStopRotX (Math|Rotator|MakeRotfromX _impactnormal))
      (Variables|Effects|SetStopSurfaceType (Variables|/Script/PhysicsCore.PhysicalMaterial|GetSurfaceType _physmat)))
    (:then_1
      (bind _hitactor (CallFunction|SpawnBounceFX (Variables|Default|GetOnBounceHitResult)))
      (Variables|Default|SetSpawnBounceFXHitActor _hitactor))
    (:then_2
      ;;[1 1 1] Adjust Gravity on bounce, is slow on bounce then explode.
      (Variables|/Script/Engine.ProjectileMovementComponent|SetProjectileGravityScale (+ (Variables|/Script/Engine.ProjectileMovementComponent|GetProjectileGravityScale _projectilemovementcomponent) (Variables|Default|GetAddedGravityPerBounce)) _projectilemovementcomponent)
      (if (<= (Math|Vector|VectorLength (Variables|/Script/Engine.MovementComponent|GetVelocity _projectilemovementcomponent)) (Variables|Default|GetMinSpeedBeforeExplode))
        (Development|PrintString "Slowed Exploed")
        (Utilities|FlowControl|DoOnce false
          (:Completed
            ;;[1 1 1] DelayExplosion
            (if (not (Delivery|IsBeingKilled))
              (Components|Activation|SetActive (Variables|Default|GetTrail) false false)
              (CallFunction|DoExplosionOnce))))))))

(event OnComponentBeginOverlap(PlayerOverlap)
  (params (OverlappedComponent :type Primitive%20Component%20Object%20Reference) (OtherActor :type Actor%20Object%20Reference) (OtherComp :type Primitive%20Component%20Object%20Reference) (OtherBodyIndex :type Integer) (bFromSweep :type Boolean) (SweepResult :type Hit%20Result%20Structure%20%28by%20ref%29))
  ;[1 1 1] Player Overlap or Hit
  (Utilities|Casting|CastToFortPlayerPawn OtherActor
    (:then
      _)
    (:CastFailed
      (Utilities|Casting|CastToFortAIPawn OtherActor
        (:then
          _)
        (:CastFailed
          (bind _returnvalue (GameplayTags|HasMatchingGameplayTag(Message) OtherActor "(TagName=\"Gameplay.TriggerProxGrenade\")"))
          (if _returnvalue
            _)))))
  (Development|PrintString "Proj Overlapped")
  (if (not (Team|OnSameTeam OtherActor self))
    (Utilities|FlowControl|DoOnce false
      (:Completed
        (CallFunction|DoExplosionOnce)))))

(event Collision|EventHit
  (params (MyComp :type Primitive%20Component%20Object%20Reference) (Other :type Actor%20Object%20Reference) (OtherComp :type Primitive%20Component%20Object%20Reference) (bSelfMoved :type Boolean) (HitLocation :type Vector :default "0, 0, 0") (HitNormal :type Vector :default "0, 0, 0") (NormalImpulse :type Vector :default "0, 0, 0") (Hit :type Hit%20Result%20Structure%20%28by%20ref%29))
  ;[1 1 1] Player Overlap or Hit
  (|CallParentFunction MyComp Other OtherComp bSelfMoved HitLocation HitNormal NormalImpulse Hit)
  (bind _asfort_player_pawn (Utilities|Casting|CastToFortPlayerPawn Other)
    (:then
      (if (not (Team|OnSameTeam _asfort_player_pawn self))
        _))
    (:CastFailed
      (bind _asfort_aipawn (Utilities|Casting|CastToFortAIPawn Other))
      (if (not (Team|OnSameTeam _asfort_aipawn self))
        _)))
  (Utilities|FlowControl|DoOnce false
    (:Completed
      (CallFunction|DoExplosionOnce))))

(event Custom|DoExplosionOnce
  ;[1 1 1] Do explosion and ensure everything that needs to be hidden or deactivated is.
  (Utilities|FlowControl|DoOnce false
    (:Completed
      (Delivery|DoExplosion))))

(event EventOnExploded
  (params (HitActors :type Array%20of%20Actor%20Object%20References) (HitResults :type Array%20of%20Hit%20Result%20Structures))
  ;[1 1 1] Do explosion and ensure everything that needs to be hidden or deactivated is.
  (|CallParentFunction HitActors HitResults)
  (SoundIndicator|StopSoundIndicatorLooping (Variables|Default|GetFortSoundIndicator))
  (Development|SetHiddeninGame (Variables|Projectile_GrenadeLauncher|GetMesh) true false)
  (Components|Activation|Deactivate (Variables|Default|GetTrail))
  (Fortnite|ShowSoundIndicatorLocal :Location (Transformation|GetActorLocation) :MaxAudibleDistance (Variables|/Script/Engine.SoundBase|GetMaxDistance (Variables|Sounds|GetGrenade_Explosion_Sound)) :IndicatorType "Gunshot" :Instigator self :Tint "(R=0.000000,G=0.000000,B=0.000000,A=0.000000)" :AttenuationMaxRangeMultiplier 1.0 :AttenuationIntensityMultiplier 1.0))

(event OnEnterWaterEvent(WaterInteractionComponent)
  (params (WaterBody :type Fort%20Water%20Body%20Actor%20Object%20Reference) (WaterInteractionComponent :type Fort%20Water%20Interaction%20Component%20Object%20Reference) (bIsFirstBody :type Boolean))
  ;[1 1 1] JLavender: FORT-389207
  ; Handle entering water. Replicate surface type for FX and then explode.
  (Utilities|FlowControl|DoOnce false
    (:Completed
      (Variables|Effects|SetStopSurfaceType "SurfaceType8")
      (CallFunction|DoExplosionOnce))))

(fn GetAffiliationsToShowFor
  (return (Utilities|Array|MakeArray "Hostile" "Neutral" "Friendly")))

(fn OnRep_RepCollision
  (if (Variables|Default|GetRepCollision)
    (Collision|SetCollisionEnabled (Variables|Default|GetPlayerOverlap) "QueryAndPhysics")))

(fn ShouldShowSoundIndicator
  (params (PlayerController :type Fort%20Player%20Controller%20Object%20Reference))
  (return true))

(fn SpawnBounceFX
  (params (Hit :type Hit%20Result%20Structure%20%28by%20ref%29))
  (bind (_bblockinghit _binitialoverlap _time _distance _location _impactpoint _normal _impactnormal _physmat _hitactor) (Collision|BreakHitResult Hit))
  (bind _returnvalue (Math|Conversions|RotationFromXVector _impactnormal))
  (bind _returnvalue_2 (Collision|GetFortPhysicalSurface Hit))
  (if (Utilities|Enum|Equal(Enum) _returnvalue_2 "SurfaceType8")
    (Effects|SpawnEmitteratLocation (Variables|Effects|GetBounce_WaterSplash_PS) _impactpoint _returnvalue)
    (else
      (Effects|SpawnEmitteratLocation (Variables|Effects|GetBounce_Generic_PS) _impactpoint _returnvalue)))
  (Audio|OnCollision (Variables|Default|GetFortCollisionAudio) _impactpoint (* _impactnormal (Variables|/Script/Engine.MovementComponent|GetVelocity (Variables|Movement|GetProjectileMovementComponent))))
  (Class|BPSurfaceTypeSoundComponent|SetSurfaceType (Variables|Default|GetBP_SurfaceTypeSoundComponent) _returnvalue_2 _hitactor)
  (return _hitactor _returnvalue_2 _returnvalue))

(fn ConstructionScript
  ;[1 1 1] Set up templates
  (Effects|Components|ParticleSystem|SetTemplate (Variables|Default|GetRibbonTrail_PSC) (Variables|Effects|GetRibbonTrail_PS))
  (Effects|Components|ParticleSystem|SetTemplate (Variables|Default|GetWaterSplash_PSC) (Variables|Effects|GetBounce_WaterSplash_PS))
  (Effects|Components|ParticleSystem|SetTemplate (Variables|Default|GetBounce_Generic_PSC) (Variables|Effects|GetBounce_Generic_PS))
  ;[1 1 1] Ensure the collision is off and set the collision scale
  (Collision|SetCollisionEnabled (Variables|Default|GetPlayerOverlap) "NoCollision")
  (Components|Sphere|SetSphereRadius (Variables|Default|GetPlayerOverlap) (ScalableFloat|GetValueatLevel (Variables|Rows|GetRow_PlayerOverlap)) true)
  ;[1 1 1] Save out FX scale
  (Variables|SetFXScale_X (Math|Vector|MakeVector (ScalableFloat|GetValueatLevel (Variables|Rows|GetRow_FXScale)) (ScalableFloat|GetValueatLevel (Variables|Rows|GetRow_FXScale)) (ScalableFloat|GetValueatLevel (Variables|Rows|GetRow_FXScale))))
  ;[1 1 1] Set Gravity
  (Variables|Default|SetAddedGravityPerBounce (ScalableFloat|GetValueatLevel (Variables|Rows|GetRow_GravPerBounce)))
  (Variables|/Script/Engine.ProjectileMovementComponent|SetProjectileGravityScale (ScalableFloat|GetValueatLevel (Variables|Rows|GetRow_InitialGrav)) (Variables|Movement|GetProjectileMovementComponent)))
